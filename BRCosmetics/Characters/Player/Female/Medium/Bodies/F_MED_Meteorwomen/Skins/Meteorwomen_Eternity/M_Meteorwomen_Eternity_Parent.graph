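# Parent material for the Meteorwomen Eternity character skin, layering skin FX between the shared 2023 CharacterParent PreFX and PostFX functions.
# An EmissiveMask result is added to emissive, then optional static-switched panning emissive, SubUV flipbook and threshold emissive stages chain onto the material attributes.
# Tech art masks select where effects apply.
# An optional glitch distorts mask UVs and blends mask channels per RGB, with an optional area mask.
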Material /BRCosmetics/Characters/Player/Female/Medium/Bodies/F_MED_Meteorwomen/Skins/Meteorwomen_Eternity/M_Meteorwomen_Eternity_Parent
graph {
 properties {
  bUseMaterialAttributes=true
  bUsedWithMeshDeformer=true
  bUsedWithMorphTargets=true
  bUsedWithSkeletalMesh=true
  pixelDepthOffsetMode=PDOM_Legacy
  shadingModel=MSM_FromMaterialExpression
  usedShadingModels=Default Lit | Subsurface | Clear Coat | Subsurface Profile | Cloth
 }
 nodes [
  n1 = TextureCoordinate()
  n2 = MaterialFunctionCall(Override Tech Art Mask UVs=StaticSwitchParameter(True=MaterialFunctionCall!(UV=@n1) {materialFunction=/Game/Weapons/WeaponSkins/MaterialFunction/MF_GlitchFunction}.Glitched UV, False=@n1) {group=Skin FX, parameterName=UseGlitch}) {materialFunction=&MF_CharacterParent_2023_PreFX}
  MaterialFunctionCall(Material=NamedRerouteUsage(), M=@n2.M, Pre Skinned Local Position=@n2.M, Distance Blend=@n2.M, Fresnel=@n2.M) {materialFunction=&MF_CharacterParent_2023_PostFX}
  n3 = TextureSampleParameter2D(UVs=@n1) {
   group=05_Tech Art Masks
   parameterName=TechArtMask
   samplerType=SAMPLERTYPE_Masks
   sortPriority=0
   texture=&T_BlackMask
  }
  NamedRerouteDeclaration(None=StaticSwitchParameter(True=@n3.RGBA, False=@n2.Tech Art Mask) {group=Skin FX, parameterName=UseTechArtMask2}) {name=TAMask, nodeColor=<rgba 0.4531 0 1 1>}
  n4 = MaterialFunctionCall(Float3=NamedRerouteUsage()) {materialFunction=&BreakOutFloat3Components}
  n5 = ScalarParameter() {group=Glitch, parameterName=Glitch Red}
  n6 = ScalarParameter() {group=Glitch, parameterName=Glitch Green}
  n7 = AppendVector(
   A=AppendVector(A=LinearInterpolate(A=@n3.R, B=@n4.R, Alpha=@n5), B=LinearInterpolate(A=@n3.G, B=@n4.G, Alpha=@n6))
   B=LinearInterpolate(A=@n3.B, B=@n4.B, Alpha=ScalarParameter() {group=Glitch, parameterName=Glitch Blue})
  )
  n8 = Multiply(A=@n7, B=@n3.R)
  n9 = ScalarParameter() {group=Glitch, parameterName=UseAreaMask}
  NamedRerouteDeclaration(None=StaticSwitchParameter(True=LinearInterpolate(A=@n7, B=@n8, Alpha=@n9), False=NamedRerouteUsage()) {group=Skin FX, parameterName=UseGlitch}) {name=TAMask2, nodeColor=<rgba 0.0312 1 0 1>}
  NamedRerouteDeclaration(None=@n2.Result) {name=MA_IN, nodeColor=<rgba 0 0.8359 1 1>}
  n10 = NamedRerouteUsage()
  n11 = ChannelMaskParameter(None=@n10) {group=Skin FX, parameterName=Simple Threshold Channel}
  n12 = ChannelMaskParameter(None=@n10) {group=Skin FX, parameterName=Simple Threshold Sequence Channel}
  n13 = NamedRerouteUsage()
  n14 = GetMaterialAttributes(None=NamedRerouteUsage()) {attributeGetTypes [#1]}
  n15 = SetMaterialAttributes(
   MaterialAttributes=@n14.MaterialAttributes
   Emissive Color=Add(
    A=@n14.Emissive Color
    B=MaterialFunctionCall!(Fresnel=@n2.Fresnel) {materialFunction=/Game/Packages/Fortress_SharedMaterials/Base_Material_Functions/EmissiveMask}.Result
   )
  ) {attributeSetTypes [#1]}
  n16 = StaticSwitchParameter(
   True=MaterialFunctionCall!(
    Alpha=StaticComponentMaskParameter(None=@n2.Tech Art Mask) {defaultR=true, group=Skin FX, parameterName=PanningEmissive_Channel}
    isVehicle
    Gmap=ComponentMask(None=@n2.M) {g=true}
    UVOffset=StaticComponentMaskParameter(None=@n2.Tech Art Mask) {defaultB=true, group=Skin FX, parameterName=PanningEmissive_UVOffset_Channel}
    Local Normal
    Local Position
    EmissiveColorOverride
    Material=@n15
   ) {materialFunction=/Game/Weapons/WeaponSkins/MaterialFunction/PanningEmissive}.Result
   False=@n15
  ) {group=Skin FX, parameterName=Use Panning Emissive}
  n17 = StaticSwitchParameter(
   True=MaterialFunctionCall!(Alpha=StaticComponentMaskParameter(None=@n13) {defaultR=true, group=Skin FX, parameterName=Flipbook Texture Channel}, Material=@n16) {materialFunction=/Game/Weapons/WeaponSkins/MaterialFunction/SubUVTextures}.Result
   False=@n16
  ) {group=Skin FX, parameterName=Use Sub UV texture}
  NamedRerouteDeclaration(None=StaticSwitchParameter(
    True=MaterialFunctionCall!(Alpha=@n11.R, Sequence=@n12.R, Material=@n17) {materialFunction=/Game/Weapons/WeaponSkins/MaterialFunction/SimpleThresholdEmissive}.Result
    False=@n17
   ) {group=Skin FX, parameterName=Use Simple Threshold}) {name=MA_OUT, nodeColor=<rgba 1 0.8438 0 1>}
  NamedRerouteUsage()
 ]
}
